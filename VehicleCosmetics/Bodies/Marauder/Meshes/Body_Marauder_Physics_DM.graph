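# Physics asset for the Marauder vehicle body cosmetic mesh, giving it a simple collision volume.
# A single body on the root bone holds one box shape, longer than it is wide, raised above the bone origin and shifted slightly along positive X.
# It has no constraints, so the asset defines no jointed or per-part physics.
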PhysicsAsset /VehicleCosmetics/Bodies/Marauder/Meshes/Body_Marauder_Physics_DM
bodies {root [{shapeType=Box, extentX=444, extentY=232, extentZ=116, shapeName=root_box, center=<24 0 89>}]}
constraints []
